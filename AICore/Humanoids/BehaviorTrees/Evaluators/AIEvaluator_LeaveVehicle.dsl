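; Behavior tree evaluator that decides whether a humanoid AI in a vehicle should leave it.
; A sequence checks the AI group leader's validity and vehicle state, as a comment says to leave if the leader is dead or has left.
; It then returns true if the blackboard vehicle carries the disabled gameplay tag or the bot controller's alert level is Threatened.
; Otherwise it returns false.
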(fn EvaluateStartingConditions
  (params (OwnerComp :type Behavior%20Tree%20Component%20Object%20Reference))
  (Utilities|FlowControl|Sequence
    (:then_0
      ;[1 1 1] Find Group
      (bind _asaicontroller (Utilities|Casting|CastToAIController (Components|GetOwner OwnerComp)))
      (bind _asfort_pawn (Utilities|Casting|CastToFortPawn (Pawn|GetControlledPawn _asaicontroller)))
      (Utilities|IsValid (Actor|GetComponentbyClass _asfort_pawn "/Script/FortniteAI.FortPawnComponent_AIGroup")
        (:Is%20Valid
          ;[1 1 1] If Group Leader is dead or has left vehicle -> Leave
          (Utilities|IsValid (AIGroup|GetGroupLeader (Actor|GetComponentbyClass _asfort_pawn "/Script/FortniteAI.FortPawnComponent_AIGroup"))
            (:Is%20Valid
              (bind _asfort_player_pawn (Utilities|Casting|CastToFortPlayerPawn (Components|GetOwner (AIGroup|GetGroupLeader (Actor|GetComponentbyClass _asfort_pawn "/Script/FortniteAI.FortPawnComponent_AIGroup"))))
                (:then
                  (if (Vehicle|IsinVehicle _asfort_player_pawn)
                    _
                    (else
                      _)))
                (:CastFailed
                  _))
              (return true))
            (:Is%20Not%20Valid
              (if (AIGroup|IsinGroup (Actor|GetComponentbyClass _asfort_pawn "/Script/FortniteAI.FortPawnComponent_AIGroup"))
                _))))))
    (:then_1
      ;[1 1 1] Vehicle is disabled
      (bind _asactor (Utilities|Casting|CastToActor (Blackboard|GetObject (Variables|Default|GetVehicle%20Actor) OwnerComp)))
      (bind _returnvalue (GameplayTags|HasAnyMatchingGameplayTags(Message) _asactor (Variables|Default|GetVehicle%20Disabled%20Tag)))
      (if _returnvalue
        (return true)))
    (:then_2
      ;[1 1 1] If AI are threatened, leave vehicle.
      (bind _asfort_athena_aibot_controller (Utilities|Casting|CastToFortAthenaAIBotController (Components|GetOwner OwnerComp)))
      (if (Utilities|Enum|Equal(Enum) (Variables|/Script/FortniteAI.FortAthenaAIBotController|GetCurrentAlertLevel _asfort_athena_aibot_controller) "Threatened")
        (return true)))
    (:then_3
      ;[1 1 1] Don't leave vehicle
      (return false))))
