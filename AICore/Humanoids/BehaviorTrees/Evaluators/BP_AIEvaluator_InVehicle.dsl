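; Behavior tree evaluator for humanoid AI that gates on whether the agent is in a vehicle.
; EvaluateStartingConditions reads the Vehicle Actor key from the owning behavior tree's blackboard and returns true when that object reference is valid.
; The check relies solely on the blackboard value and does not query the pawn or vehicle directly.
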(fn EvaluateStartingConditions
  (params (OwnerComp :type Behavior%20Tree%20Component%20Object%20Reference))
  (return (Utilities|IsValid (Blackboard|GetObject (Variables|Default|GetVehicle%20Actor) OwnerComp) :!class /Script/Engine.KismetSystemLibrary)))
