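; AI evaluator for humanoids that decides whether the vehicle-driver combat behavior can start.
; EvaluateStartingConditions reads the behavior tree blackboard, requiring a valid vehicle actor and the driving-vehicle flag.
; It then switches on the alert level, where only LostTarget returns true.
; Despite the "Is In Combat" label, Threatened has no branch of its own that returns true.
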(fn EvaluateStartingConditions
  (params (OwnerComp :type Behavior%20Tree%20Component%20Object%20Reference))
  ;[1 1 1] Is In Vehicle
  (if (Utilities|IsValid (Blackboard|GetObject (Variables|Default|GetVehicle%20Actor) OwnerComp) :!class /Script/Engine.KismetSystemLibrary)
    ;[1 1 1] Is Driver
    (if (Blackboard|GetBool (Variables|Default|GetIsDrivingVehicle) OwnerComp)
      ;[1 1 1] Is In Combat
      (switch Utilities|FlowControl|Switch|SwitchonEAlertLevel (Utilities|Enum|BytetoEnumEAlertLevel (Blackboard|GetEnum (Variables|Default|GetAlertLevel) OwnerComp))
        (:Unaware
          _)
        (:Suspicious
          _)
        (:LostTarget
          (return true))
        (:Threatened
          _)
        (:Count
          _))
      (return false)
      (else
        (return false)))
    (else
      (return false))))
